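# Niagara effect for the DrapeUltra backpack cosmetic, drawing glowing meshes and an orange particle light.
# One emitter burst-spawns two non-expiring particles, using partition index to switch each one's scale, offset, rotation and VisibilityTag across four mesh renderers and a light renderer.
# Per-partition waveforms at slightly different frequencies pulse color between two warm tints.
# The offset light derives its color from particle color, so it follows that pulse.
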NiagaraSystem /BRCosmetics/Accessories/FORT_Backpacks/Backpack_F_MED_DrapeUltra/FX/NS_Backpack_DrapeUltra
dependencies {usedDataInterfaces [], usedDynamicInputs [], usedModules [&SystemState]}
emitters {NE_Backpack_DrapeUltra {
  emitterUpdateScript [
   &EmitterState(
    Inactive Response=::NewEnumerator1
    Life Cycle Mode=::NewEnumerator1
    Loop Behavior=::NewEnumerator1
    Loop Duration Mode=::NewEnumerator1
    UseLoopDelay=false
   )
   &SpawnBurst_Instantaneous(Spawn Count=2, Use Spawn Probability=false, UseLoopCountLimit=false)
  ]
  particleSpawnScript [
   &PartitionParticles()
   &InitializeParticle(
    Lifetime=0
    Mesh Scale Mode=::NewEnumerator3
    Mesh Uniform Scale=&MergeHelper_FloatSwitch(0=0.9, 1=0.47, 2=?, 3=?, Index=$Particles.PartitionParticles.Partition, Num=::NewEnumerator0)
    Position Offset=&MergeHelper_VectorSwitch(0=<0 0 0>, 1=<0 -5.5 0.5>, 2=?, 3=?, Index=$Particles.PartitionParticles.Partition, Num=::NewEnumerator0)
    Sprite Size Mode=::NewEnumerator3
    Uniform Sprite Size=25
    UsePositionOffset=true
    Write Material Random=true
   )
   /SetVariables(Particles.VisibilityTag=$Particles.PartitionParticles.Partition) {isSetParameters=true}
   &InitialMeshOrientation(
    Mesh Axis To Orient=<0 0 0>
    Mesh Orientation Mode=::NewEnumerator2
    Rotation=&MergeHelper_VectorSwitch(0=<0 0 0>, 1=<0.19 0 0>, 2=?, 3=?, Index=$Particles.PartitionParticles.Partition, Num=::NewEnumerator0)
    Use Rotation Vector=true
   )
  ]
  particleUpdateScript [
   &ParticleState(Kill Particles When Lifetime Has Elapsed=false, Loop Particles Lifetime=true)
   &ScaleColor(
    Scale RGB=&MergeHelper_VectorSwitch(
     0=&Lerp_Vector(
      A=&MakeVectorFromLinearColorRGB(Color=<rgba 0.9 0.72 0.72 1>)
      B=&MakeVectorFromLinearColorRGB(Color=<rgba 1.07 1.07 0.9095 1>)
      Alpha=&Waveform("Waveform [1]"=::NewEnumerator12, "[1] Frequency"=1.3, "[1] Amplitude Min/Max"=<0 1>)
     )
     1=&Lerp_Vector(
      A=&MakeVectorFromLinearColorRGB(Color=<rgba 0.85 0.68 0.68 1>)
      B=&MakeVectorFromLinearColorRGB(Color=<rgba 1.05 1.05 0.8925 1>)
      Alpha=&Waveform("Waveform [1]"=::NewEnumerator12, "[1] Frequency"=1.4, "[1] Amplitude Min/Max"=<0 1>)
     )
     2=?
     3=?
     Index=$Particles.PartitionParticles.Partition
     Num=::NewEnumerator0
    )
    ScaleA=true
    ScaleRGB=true
   )
   &FN_Light_Attributes(
    Color=&MultiplyLinearColor(Linear Color=<rgba 1 0.5 0 1>, Scale Factor=&MakeVector4_FromLinearColor(LinearColor=$Particles.Color))
    Create Color Attribute=true
    Create Exponent Attribute=true
    Create Position Attribute=true
    Create Radius Attribute=true
    Enable Particle Light Enable=true
    Exponent=2
    Position=&AddVectorToPosition(Position=$Particles.Position, Vector=<0 -15 0>)
    Radius=32
    Scale Color=2
   )
  ]
  rendererClasses [&NiagaraMeshRendererProperties, &NiagaraLightRendererProperties]
  renderers [
   {rendererClass=&NiagaraMeshRendererProperties, rendererIndex=0}
   {rendererClass=&NiagaraMeshRendererProperties, rendererIndex=1}
   {rendererClass=&NiagaraLightRendererProperties, rendererIndex=2}
   {rendererClass=&NiagaraMeshRendererProperties, rendererIndex=3}
   {rendererClass=&NiagaraMeshRendererProperties, rendererIndex=4}
  ]
 }}
summary {
 systemName=NS_Backpack_DrapeUltra
 emitters [{
   emitterName=NE_Backpack_DrapeUltra
   rendererClasses [&NiagaraMeshRendererProperties, &NiagaraLightRendererProperties]
  }]
 userVariables []
}
