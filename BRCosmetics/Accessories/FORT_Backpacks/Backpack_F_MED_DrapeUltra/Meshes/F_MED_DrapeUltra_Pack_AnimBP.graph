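# Anim Blueprint for the DrapeUltra backpack that shows or hides a material-driven element of the pack based on pickaxe and movement state.
# The AnimGraph passes the input pose through unchanged.
# Weapon, sliding, tactical sprint and skydiving events set "Hide Element 01" on a dynamic material instance, fetched only on weapon change.
# The element hides while a DrapeUltra-tagged pickaxe is equipped, except during skydiving or non-sliding tactical sprint.
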AnimBlueprint /BRCosmetics/Accessories/FORT_Backpacks/Backpack_F_MED_DrapeUltra/Meshes/F_MED_DrapeUltra_Pack_AnimBP
target_skeleton=&/F_MED_DrapeUltra_Pack_Skeleton
graphs [
 AnimGraph {nodes [Output Pose(Result=Input Pose()) {layerGroup=None}]}
 EventGraph {nodes [
   n1 = Event OnBeginSliding() {is=title K2Node_Event}
   n2 = PickaxeItemDefinitionHasAnyMetaTag(Tags=Make Array("[0]"=(TagName="Cosmetics.MetaTags.DrapeUltra")), bRequiresEquipped=true)
   n3 = Branch(execute=[@n7.then, @n1.then, @n14.then], Condition=@n2)
   n4 = NOT Boolean(A=Get bIsSliding())
   n5 = Get bIsTacticalSprinting()
   > n6 = Branch(Condition=OR Boolean(A=Get bIsSkyDiving(), B=AND Boolean(A=@n5, B=@n4)))
   n7 = Get Dynamic Material Instance(execute=@n8.then, SkeletalMeshComponent=GetPartSkeletalMeshComponent())
   SetScalarParameterValue(execute=[@n12.then, @n13.then, @n10.else, @n6.then, @n3.else], self=@n7, ParameterName=Hide Element 01, Value=0)
   SetScalarParameterValue(execute=[@n10.then, @n6.else], self=@n7, ParameterName=Hide Element 01, Value=1)
   n8 = Event OnWeaponChanged()
   n9 = Event OnEndTacticalSprinting() {is=title K2Node_Event}
   n10 = Branch(execute=[@n9.then, @n11.then], Condition=@n2)
   n11 = Event OnEndSkydiving()
   n12 = Event OnBeginTacticalSprinting() {is=title K2Node_Event}
   n13 = Event OnBeginSkydiving()
   n14 = Event OnEndSliding() {is=title K2Node_Event}
   Reroute Node(InputPin:Wildcard) -> (OutputPin:Wildcard)
  ]}
]
variables []
